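# Physics asset for the Hamster pet backpack cosmetic's skeletal mesh, giving its bottle and tray bones simulated secondary motion.
# A box body on the root covers the pack, while capsule bodies sit on dyn_bottle and dyn_tray.
# Both dynamic bodies are constrained directly to root: the tray is held nearly rigid, while the bottle gets much wider swing and twist limits so it can dangle.
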PhysicsAsset /BRCosmetics/Accessories/FORT_Backpacks/Backpack_Pets_Hamster/Meshes/HamsterPack_Physics
bodies {
 dyn_bottle [{shapeType=Capsule, length=13.5979, radius=6.7184, center=<5.1189 0.2187 -12.9063>}]
 dyn_tray [{shapeType=Capsule, length=6.1519, radius=3.8294, center=<0.0077 -2.9278 -1.6109>, rotation=<rot 0 -89.9613 -89.5761>}]
 root [{shapeType=Box, extentX=27.5827, extentY=14.655, extentZ=29.9034, center=<0 -11.856 -4.6713>}]
}
constraints [
 {bone1Name=dyn_tray, bone2Name=root, swing1LimitDegrees=1, swing2LimitDegrees=1, twistLimitDegrees=5}
 {bone1Name=dyn_bottle, bone2Name=root, swing1LimitDegrees=45, swing2LimitDegrees=45, twistLimitDegrees=15}
]
