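# Animation Blueprint for the hamster pet backpack, animating it in response to the wearing player's state.
# Each update copies movement, crouch, targeting and hunker flags from the player's AnimBP, forcing stationary crouch in vehicles.
# These and inherited flags drive a state machine, with notify-triggered sprint breaks and an unconnected idle-break stub.
# Its pose feeds root-relative RigidBody physics, then an additive bone tweak offsets dyn_bottle while skydiving.
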AnimBlueprint /BRCosmetics/Accessories/FORT_Backpacks/Backpack_Pets_Hamster/Meshes/HamsterPack_AnimBP
target_skeleton=&/HamsterPack_Skeleton
graphs [
 AnimGraph {nodes [
   AnimGraphNode_StateMachine_0 = HamPack_Anims() {is=title AnimGraphNode_StateMachine}
   n1 = RigidBody(ComponentPose=Local To Component(LocalPose=@AnimGraphNode_StateMachine_0)) {
    bOverrideWorldGravity=true
    overridePhysicsAsset=&/HamsterPack_Physics
    simulationSpace=BaseBoneSpace
    alphaBoolBlend {blendOption=Linear}
    baseBoneRef {boneName=root}
    componentAppliedLinearAccClamp=<500 500 500>
    componentLinearAccScale=<0.5 0.5 0.5>
    componentLinearVelScale=<0.5 0.5 0.5>
    overrideWorldGravity {z=-980}
   }
   n2 = "Transform (Modify) Bone - Bone: dyn_bottle"!(ComponentPose=@n1, bAlphaBoolEnabled:bool=Get bIsSkyDiving(), Translation="0,-3,0", Rotation="0,-50,0", Scale="1,1,1") -> (Pose) {
    alphaInputType=Bool
    rotationMode=BMM_Additive
    translationMode=BMM_Additive
    alphaBoolBlend {blendOption=Linear}
    boneToModify {boneName=dyn_bottle}
    rotation {yaw=-50}
    translation {y=-3}
   }
   Output Pose(Result=Component To Local(ComponentPose=@n2)) {layerGroup=None}
  ]}
 HamPack_Anims {
  entry=Idle
  path=AnimGraph.AnimGraphNode_StateMachine_0.HamPack_Anims
  ids=id
  nodes [
   AnimStateNode_0 = Idle()
   AnimStateNode_1 = Sprint() {is=title AnimStateNode}
   AnimStateNode_10 = HunkerToIdle() {is=title AnimStateNode}
   AnimStateNode_11 = CrouchADS() {is=title AnimStateNode}
   AnimStateNode_12 = ADSToIdle() {is=title AnimStateNode}
   AnimStateNode_13 = InAirHunker() {is=title AnimStateNode}
   AnimStateNode_16 = ToSkyDiving() {is=title AnimStateNode}
   AnimStateNode_18 = SkyDiving() {is=title AnimStateNode}
   AnimStateNode_2 = ToCrouch() {is=title AnimStateNode}
   AnimStateNode_3 = SprintToIdle() {is=title AnimStateNode}
   AnimStateNode_4 = CrouchIdle() {is=title AnimStateNode}
   AnimStateNode_5 = CrouchToIdle() {is=title AnimStateNode}
   AnimStateNode_6 = ADS() {is=title AnimStateNode}
   AnimStateNode_7 = CrouchLocomotion() {is=title AnimStateNode}
   AnimStateNode_8 = To CrouchIdle() {is=title AnimStateNode}
   AnimStateNode_9 = Hunker() {is=title AnimStateNode}
   AnimStateTransitionNode_0 = AnimStateTransitionNode()
   AnimStateTransitionNode_1 = AnimStateTransitionNode()
   AnimStateTransitionNode_10 = AnimStateTransitionNode()
   AnimStateTransitionNode_11 = AnimStateTransitionNode()
   AnimStateTransitionNode_12 = AnimStateTransitionNode()
   AnimStateTransitionNode_13 = AnimStateTransitionNode()
   AnimStateTransitionNode_14 = AnimStateTransitionNode()
   AnimStateTransitionNode_15 = AnimStateTransitionNode()
   AnimStateTransitionNode_17 = AnimStateTransitionNode()
   AnimStateTransitionNode_18 = AnimStateTransitionNode() {bAutomaticRuleBasedOnSequencePlayerInState=true}
   AnimStateTransitionNode_19 = AnimStateTransitionNode()
   AnimStateTransitionNode_2 = AnimStateTransitionNode()
   AnimStateTransitionNode_20 = AnimStateTransitionNode()
   AnimStateTransitionNode_21 = AnimStateTransitionNode() {bAutomaticRuleBasedOnSequencePlayerInState=true}
   AnimStateTransitionNode_22 = AnimStateTransitionNode()
   AnimStateTransitionNode_26 = AnimStateTransitionNode() {bAutomaticRuleBasedOnSequencePlayerInState=true}
   AnimStateTransitionNode_27 = AnimStateTransitionNode()
   AnimStateTransitionNode_28 = AnimStateTransitionNode()
   AnimStateTransitionNode_29 = AnimStateTransitionNode()
   AnimStateTransitionNode_3 = AnimStateTransitionNode() {bAutomaticRuleBasedOnSequencePlayerInState=true}
   AnimStateTransitionNode_30 = AnimStateTransitionNode()
   AnimStateTransitionNode_31 = AnimStateTransitionNode() {bAutomaticRuleBasedOnSequencePlayerInState=true}
   AnimStateTransitionNode_33 = AnimStateTransitionNode()
   AnimStateTransitionNode_34 = AnimStateTransitionNode()
   AnimStateTransitionNode_35 = AnimStateTransitionNode()
   AnimStateTransitionNode_36 = AnimStateTransitionNode()
   AnimStateTransitionNode_37 = AnimStateTransitionNode()
   AnimStateTransitionNode_4 = AnimStateTransitionNode()
   AnimStateTransitionNode_41 = AnimStateTransitionNode() {bAutomaticRuleBasedOnSequencePlayerInState=true}
   AnimStateTransitionNode_42 = AnimStateTransitionNode()
   AnimStateTransitionNode_43 = AnimStateTransitionNode()
   AnimStateTransitionNode_5 = AnimStateTransitionNode()
   AnimStateTransitionNode_6 = AnimStateTransitionNode() {bAutomaticRuleBasedOnSequencePlayerInState=true}
   AnimStateTransitionNode_7 = AnimStateTransitionNode()
   AnimStateTransitionNode_8 = AnimStateTransitionNode()
   AnimStateTransitionNode_9 = AnimStateTransitionNode()
  ]
  links [
   {
    dst=ADSToIdle
    rule=AnimGraph.AnimGraphNode_StateMachine_0.HamPack_Anims.AnimStateTransitionNode_7.Transition
    src=ADS
    via=AnimStateTransitionNode_7
   }
   {
    dst=Idle
    rule=AnimGraph.AnimGraphNode_StateMachine_0.HamPack_Anims.AnimStateTransitionNode_14.Transition
    src=ADSToIdle
    via=AnimStateTransitionNode_14
   }
   {
    dst=Sprint
    rule=AnimGraph.AnimGraphNode_StateMachine_0.HamPack_Anims.AnimStateTransitionNode_18.Transition
    src=ADSToIdle
    via=AnimStateTransitionNode_18
   }
   {
    dst=CrouchLocomotion
    rule=AnimGraph.AnimGraphNode_StateMachine_0.HamPack_Anims.AnimStateTransitionNode_35.Transition
    src=CrouchADS
    via=AnimStateTransitionNode_35
   }
   {
    dst=CrouchADS
    rule=AnimGraph.AnimGraphNode_StateMachine_0.HamPack_Anims.AnimStateTransitionNode_36.Transition
    src=CrouchIdle
    via=AnimStateTransitionNode_36
   }
   {
    dst=CrouchLocomotion
    rule=AnimGraph.AnimGraphNode_StateMachine_0.HamPack_Anims.AnimStateTransitionNode_19.Transition
    src=CrouchIdle
    via=AnimStateTransitionNode_19
   }
   {
    dst=CrouchToIdle
    rule=AnimGraph.AnimGraphNode_StateMachine_0.HamPack_Anims.AnimStateTransitionNode_9.Transition
    src=CrouchIdle
    via=AnimStateTransitionNode_9
   }
   {
    dst=CrouchADS
    rule=AnimGraph.AnimGraphNode_StateMachine_0.HamPack_Anims.AnimStateTransitionNode_34.Transition
    src=CrouchLocomotion
    via=AnimStateTransitionNode_34
   }
   {
    dst=CrouchToIdle
    rule=AnimGraph.AnimGraphNode_StateMachine_0.HamPack_Anims.AnimStateTransitionNode_33.Transition
    src=CrouchLocomotion
    via=AnimStateTransitionNode_33
   }
   {
    dst=To CrouchIdle
    rule=AnimGraph.AnimGraphNode_StateMachine_0.HamPack_Anims.AnimStateTransitionNode_20.Transition
    src=CrouchLocomotion
    via=AnimStateTransitionNode_20
   }
   {
    dst=CrouchIdle
    rule=AnimGraph.AnimGraphNode_StateMachine_0.HamPack_Anims.AnimStateTransitionNode_27.Transition
    src=CrouchToIdle
    via=AnimStateTransitionNode_27
   }
   {
    dst=Idle
    rule=AnimGraph.AnimGraphNode_StateMachine_0.HamPack_Anims.AnimStateTransitionNode_26.Transition
    src=CrouchToIdle
    via=AnimStateTransitionNode_26
   }
   {
    dst=HunkerToIdle
    rule=AnimGraph.AnimGraphNode_StateMachine_0.HamPack_Anims.AnimStateTransitionNode_30.Transition
    src=Hunker
    via=AnimStateTransitionNode_30
   }
   {
    dst=Idle
    rule=AnimGraph.AnimGraphNode_StateMachine_0.HamPack_Anims.AnimStateTransitionNode_31.Transition
    src=HunkerToIdle
    via=AnimStateTransitionNode_31
   }
   {
    dst=CrouchIdle
    rule=AnimGraph.AnimGraphNode_StateMachine_0.HamPack_Anims.AnimStateTransitionNode_11.Transition
    src=Idle
    via=AnimStateTransitionNode_11
   }
   {
    dst=Hunker
    rule=AnimGraph.AnimGraphNode_StateMachine_0.HamPack_Anims.AnimStateTransitionNode_28.Transition
    src=Idle
    via=AnimStateTransitionNode_28
   }
   {
    dst=InAirHunker
    rule=AnimGraph.AnimGraphNode_StateMachine_0.HamPack_Anims.AnimStateTransitionNode_10.Transition
    src=Idle
    via=AnimStateTransitionNode_10
   }
   {
    dst=Sprint
    rule=AnimGraph.AnimGraphNode_StateMachine_0.HamPack_Anims.AnimStateTransitionNode_0.Transition
    src=Idle
    via=AnimStateTransitionNode_0
   }
   {
    dst=ToSkyDiving
    rule=AnimGraph.AnimGraphNode_StateMachine_0.HamPack_Anims.AnimStateTransitionNode_29.Transition
    src=Idle
    via=AnimStateTransitionNode_29
   }
   {
    dst=Idle
    rule=AnimGraph.AnimGraphNode_StateMachine_0.HamPack_Anims.AnimStateTransitionNode_13.Transition
    src=InAirHunker
    via=AnimStateTransitionNode_13
   }
   {
    dst=Idle
    rule=AnimGraph.AnimGraphNode_StateMachine_0.HamPack_Anims.AnimStateTransitionNode_42.Transition
    src=SkyDiving
    via=AnimStateTransitionNode_42
   }
   {
    dst=ADS
    rule=AnimGraph.AnimGraphNode_StateMachine_0.HamPack_Anims.AnimStateTransitionNode_4.Transition
    src=Sprint
    via=AnimStateTransitionNode_4
   }
   {
    dst=InAirHunker
    rule=AnimGraph.AnimGraphNode_StateMachine_0.HamPack_Anims.AnimStateTransitionNode_8.Transition
    src=Sprint
    via=AnimStateTransitionNode_8
   }
   {
    dst=SprintToIdle
    rule=AnimGraph.AnimGraphNode_StateMachine_0.HamPack_Anims.AnimStateTransitionNode_15.Transition
    src=Sprint
    via=AnimStateTransitionNode_15
   }
   {
    dst=SprintToIdle
    rule=AnimGraph.AnimGraphNode_StateMachine_0.HamPack_Anims.AnimStateTransitionNode_2.Transition
    src=Sprint
    via=AnimStateTransitionNode_2
   }
   {
    dst=ToCrouch
    rule=AnimGraph.AnimGraphNode_StateMachine_0.HamPack_Anims.AnimStateTransitionNode_1.Transition
    src=Sprint
    via=AnimStateTransitionNode_1
   }
   {
    dst=ToSkyDiving
    rule=AnimGraph.AnimGraphNode_StateMachine_0.HamPack_Anims.AnimStateTransitionNode_43.Transition
    src=Sprint
    via=AnimStateTransitionNode_43
   }
   {
    dst=Idle
    rule=AnimGraph.AnimGraphNode_StateMachine_0.HamPack_Anims.AnimStateTransitionNode_3.Transition
    src=SprintToIdle
    via=AnimStateTransitionNode_3
   }
   {
    dst=Sprint
    rule=AnimGraph.AnimGraphNode_StateMachine_0.HamPack_Anims.AnimStateTransitionNode_12.Transition
    src=SprintToIdle
    via=AnimStateTransitionNode_12
   }
   {
    dst=CrouchIdle
    rule=AnimGraph.AnimGraphNode_StateMachine_0.HamPack_Anims.AnimStateTransitionNode_21.Transition
    src=To CrouchIdle
    via=AnimStateTransitionNode_21
   }
   {
    dst=CrouchLocomotion
    rule=AnimGraph.AnimGraphNode_StateMachine_0.HamPack_Anims.AnimStateTransitionNode_22.Transition
    src=To CrouchIdle
    via=AnimStateTransitionNode_22
   }
   {
    dst=CrouchToIdle
    rule=AnimGraph.AnimGraphNode_StateMachine_0.HamPack_Anims.AnimStateTransitionNode_37.Transition
    src=To CrouchIdle
    via=AnimStateTransitionNode_37
   }
   {
    dst=CrouchIdle
    rule=AnimGraph.AnimGraphNode_StateMachine_0.HamPack_Anims.AnimStateTransitionNode_6.Transition
    src=ToCrouch
    via=AnimStateTransitionNode_6
   }
   {
    dst=CrouchLocomotion
    rule=AnimGraph.AnimGraphNode_StateMachine_0.HamPack_Anims.AnimStateTransitionNode_17.Transition
    src=ToCrouch
    via=AnimStateTransitionNode_17
   }
   {
    dst=CrouchToIdle
    rule=AnimGraph.AnimGraphNode_StateMachine_0.HamPack_Anims.AnimStateTransitionNode_5.Transition
    src=ToCrouch
    via=AnimStateTransitionNode_5
   }
   {
    dst=SkyDiving
    rule=AnimGraph.AnimGraphNode_StateMachine_0.HamPack_Anims.AnimStateTransitionNode_41.Transition
    src=ToSkyDiving
    via=AnimStateTransitionNode_41
   }
  ]
 }
 Idle {
  path=AnimGraph.AnimGraphNode_StateMachine_0.HamPack_Anims.AnimStateNode_0.Idle
  nodes [
   AnimGraphNode_StateMachine_0 = IdleBreaks() {is=title AnimGraphNode_StateMachine}
   Output Animation Pose(Result=Sequence Player 'HamPack_Idle'() {sequence=/BRCosmetics/Accessories/FORT_Backpacks/Backpack_Pets_Hamster/Animation/HamPack_Idle})
   "State Weight (IdleBreak_01)"!() -> (ReturnValue:float) {is=title K2Node_AnimGetter}
  ]
 }
 IdleBreaks {
  entry=Chooser
  path=AnimGraph.AnimGraphNode_StateMachine_0.HamPack_Anims.AnimStateNode_0.Idle.AnimGraphNode_StateMachine_0.IdleBreaks
  ids=id
  nodes [
   AnimStateNode_0 = IdleBreak_01_1() {is=title AnimStateNode}
   AnimStateNode_1 = Chooser() {is=title AnimStateNode}
   AnimStateNode_2 = IdleBreak_01() {is=title AnimStateNode}
   AnimStateTransitionNode_5 = AnimStateTransitionNode()
   AnimStateTransitionNode_6 = AnimStateTransitionNode()
  ]
  links [
   {
    dst=IdleBreak_01
    rule=AnimGraph.AnimGraphNode_StateMachine_0.HamPack_Anims.AnimStateNode_0.Idle.AnimGraphNode_StateMachine_0.IdleBreaks.AnimStateTransitionNode_5.Transition
    src=Chooser
    via=AnimStateTransitionNode_5
   }
   {
    dst=IdleBreak_01_1
    rule=AnimGraph.AnimGraphNode_StateMachine_0.HamPack_Anims.AnimStateNode_0.Idle.AnimGraphNode_StateMachine_0.IdleBreaks.AnimStateTransitionNode_6.Transition
    src=Chooser
    via=AnimStateTransitionNode_6
   }
  ]
 }
 IdleBreak_01_1 {
  path=AnimGraph.AnimGraphNode_StateMachine_0.HamPack_Anims.AnimStateNode_0.Idle.AnimGraphNode_StateMachine_0.IdleBreaks.AnimStateNode_0.IdleBreak_01_1
  nodes [Output Animation Pose()]
 }
 Chooser {
  path=AnimGraph.AnimGraphNode_StateMachine_0.HamPack_Anims.AnimStateNode_0.Idle.AnimGraphNode_StateMachine_0.IdleBreaks.AnimStateNode_1.Chooser
  nodes [Output Animation Pose()]
 }
 IdleBreak_01 {
  path=AnimGraph.AnimGraphNode_StateMachine_0.HamPack_Anims.AnimStateNode_0.Idle.AnimGraphNode_StateMachine_0.IdleBreaks.AnimStateNode_2.IdleBreak_01
  nodes [Output Animation Pose()]
 }
 Transition {
  path=AnimGraph.AnimGraphNode_StateMachine_0.HamPack_Anims.AnimStateNode_0.Idle.AnimGraphNode_StateMachine_0.IdleBreaks.AnimStateTransitionNode_5.Transition
  nodes [Result()]
 }
 Transition {
  path=AnimGraph.AnimGraphNode_StateMachine_0.HamPack_Anims.AnimStateNode_0.Idle.AnimGraphNode_StateMachine_0.IdleBreaks.AnimStateTransitionNode_6.Transition
  nodes [Result()]
 }
 Sprint {
  path=AnimGraph.AnimGraphNode_StateMachine_0.HamPack_Anims.AnimStateNode_1.Sprint
  nodes [
   Sequence Player 'HamPack_LocomotionPose'() {sequence=/BRCosmetics/Accessories/FORT_Backpacks/Backpack_Pets_Hamster/Animation/HamPack_LocomotionPose}
   AnimGraphNode_StateMachine_0 = SprintWithBreaks() {is=title AnimGraphNode_StateMachine}
   Output Animation Pose(Result=@AnimGraphNode_StateMachine_0)
  ]
 }
 SprintWithBreaks {
  entry=Sprint
  path=AnimGraph.AnimGraphNode_StateMachine_0.HamPack_Anims.AnimStateNode_1.Sprint.AnimGraphNode_StateMachine_0.SprintWithBreaks
  ids=id
  nodes [
   AnimStateNode_0 = Sprint()
   AnimStateNode_1 = SprintBreak1() {is=title AnimStateNode}
   AnimStateNode_2 = SprintBreak2() {is=title AnimStateNode}
   AnimStateTransitionNode_0 = AnimStateTransitionNode()
   AnimStateTransitionNode_1 = AnimStateTransitionNode()
   AnimStateTransitionNode_2 = AnimStateTransitionNode()
   AnimStateTransitionNode_3 = AnimStateTransitionNode() {bAutomaticRuleBasedOnSequencePlayerInState=true}
   AnimStateTransitionNode_7 = AnimStateTransitionNode()
   AnimStateTransitionNode_8 = AnimStateTransitionNode() {bAutomaticRuleBasedOnSequencePlayerInState=true}
  ]
  links [
   {
    dst=SprintBreak1
    rule=AnimGraph.AnimGraphNode_StateMachine_0.HamPack_Anims.AnimStateNode_1.Sprint.AnimGraphNode_StateMachine_0.SprintWithBreaks.AnimStateTransitionNode_0.Transition
    src=Sprint
    via=AnimStateTransitionNode_0
   }
   {
    dst=SprintBreak2
    rule=AnimGraph.AnimGraphNode_StateMachine_0.HamPack_Anims.AnimStateNode_1.Sprint.AnimGraphNode_StateMachine_0.SprintWithBreaks.AnimStateTransitionNode_1.Transition
    src=Sprint
    via=AnimStateTransitionNode_1
   }
   {
    dst=Sprint
    rule=AnimGraph.AnimGraphNode_StateMachine_0.HamPack_Anims.AnimStateNode_1.Sprint.AnimGraphNode_StateMachine_0.SprintWithBreaks.AnimStateTransitionNode_3.Transition
    src=SprintBreak1
    via=AnimStateTransitionNode_3
   }
   {
    dst=Sprint
    rule=AnimGraph.AnimGraphNode_StateMachine_0.HamPack_Anims.AnimStateNode_1.Sprint.AnimGraphNode_StateMachine_0.SprintWithBreaks.AnimStateTransitionNode_7.Transition
    src=SprintBreak1
    via=AnimStateTransitionNode_7
   }
   {
    dst=Sprint
    rule=AnimGraph.AnimGraphNode_StateMachine_0.HamPack_Anims.AnimStateNode_1.Sprint.AnimGraphNode_StateMachine_0.SprintWithBreaks.AnimStateTransitionNode_2.Transition
    src=SprintBreak2
    via=AnimStateTransitionNode_2
   }
   {
    dst=Sprint
    rule=AnimGraph.AnimGraphNode_StateMachine_0.HamPack_Anims.AnimStateNode_1.Sprint.AnimGraphNode_StateMachine_0.SprintWithBreaks.AnimStateTransitionNode_8.Transition
    src=SprintBreak2
    via=AnimStateTransitionNode_8
   }
  ]
 }
 Sprint {
  path=AnimGraph.AnimGraphNode_StateMachine_0.HamPack_Anims.AnimStateNode_1.Sprint.AnimGraphNode_StateMachine_0.SprintWithBreaks.AnimStateNode_0.Sprint
  nodes [Output Animation Pose(Result=Sequence Player 'HamPack_LocomotionPose'() {sequence=&HamPack_LocomotionPose})]
 }
 SprintBreak1 {
  path=AnimGraph.AnimGraphNode_StateMachine_0.HamPack_Anims.AnimStateNode_1.Sprint.AnimGraphNode_StateMachine_0.SprintWithBreaks.AnimStateNode_1.SprintBreak1
  nodes [Output Animation Pose(Result=Sequence Player 'HamPack_SprintBreak_01'() {sequence=/BRCosmetics/Accessories/FORT_Backpacks/Backpack_Pets_Hamster/Animation/HamPack_SprintBreak_01})]
 }
 SprintBreak2 {
  path=AnimGraph.AnimGraphNode_StateMachine_0.HamPack_Anims.AnimStateNode_1.Sprint.AnimGraphNode_StateMachine_0.SprintWithBreaks.AnimStateNode_2.SprintBreak2
  nodes [Output Animation Pose(Result=Sequence Player 'HamPack_SprintBreak_01'() {sequence=&HamPack_SprintBreak_01})]
 }
 Transition {
  path=AnimGraph.AnimGraphNode_StateMachine_0.HamPack_Anims.AnimStateNode_1.Sprint.AnimGraphNode_StateMachine_0.SprintWithBreaks.AnimStateTransitionNode_0.Transition
  nodes [Result(bCanEnterTransition=Get PlaySprintBreak1() -> (PlaySprintBreak1:bool))]
 }
 Transition {
  path=AnimGraph.AnimGraphNode_StateMachine_0.HamPack_Anims.AnimStateNode_1.Sprint.AnimGraphNode_StateMachine_0.SprintWithBreaks.AnimStateTransitionNode_1.Transition
  nodes [Result(bCanEnterTransition=Get PlaySprintBreak2() -> (PlaySprintBreak2:bool))]
 }
 Transition {
  path=AnimGraph.AnimGraphNode_StateMachine_0.HamPack_Anims.AnimStateNode_1.Sprint.AnimGraphNode_StateMachine_0.SprintWithBreaks.AnimStateTransitionNode_2.Transition
  nodes [Result(bCanEnterTransition=NOT Boolean(A=Get PlaySprintBreak2()))]
 }
 Transition {
  path=AnimGraph.AnimGraphNode_StateMachine_0.HamPack_Anims.AnimStateNode_1.Sprint.AnimGraphNode_StateMachine_0.SprintWithBreaks.AnimStateTransitionNode_3.Transition
  nodes [Result()]
 }
 Transition {
  path=AnimGraph.AnimGraphNode_StateMachine_0.HamPack_Anims.AnimStateNode_1.Sprint.AnimGraphNode_StateMachine_0.SprintWithBreaks.AnimStateTransitionNode_7.Transition
  nodes [Result(bCanEnterTransition=NOT Boolean(A=Get PlaySprintBreak1()))]
 }
 Transition {
  path=AnimGraph.AnimGraphNode_StateMachine_0.HamPack_Anims.AnimStateNode_1.Sprint.AnimGraphNode_StateMachine_0.SprintWithBreaks.AnimStateTransitionNode_8.Transition
  nodes [Result()]
 }
 HunkerToIdle {
  path=AnimGraph.AnimGraphNode_StateMachine_0.HamPack_Anims.AnimStateNode_10.HunkerToIdle
  nodes [Output Animation Pose(Result=Sequence Player 'HamPack_LocoToCrouchIdle'() {sequence=/BRCosmetics/Accessories/FORT_Backpacks/Backpack_Pets_Hamster/Animation/HamPack_LocoToCrouchIdle})]
 }
 CrouchADS {
  path=AnimGraph.AnimGraphNode_StateMachine_0.HamPack_Anims.AnimStateNode_11.CrouchADS
  nodes [Output Animation Pose(Result=Sequence Player 'HamPack_Idle'() {sequence=&HamPack_Idle})]
 }
 ADSToIdle {
  path=AnimGraph.AnimGraphNode_StateMachine_0.HamPack_Anims.AnimStateNode_12.ADSToIdle
  nodes [Output Animation Pose(Result=Sequence Player 'HamPack_ADS_TO_Idle'() {sequence=/BRCosmetics/Accessories/FORT_Backpacks/Backpack_Pets_Hamster/Animation/HamPack_ADS_TO_Idle})]
 }
 InAirHunker {
  path=AnimGraph.AnimGraphNode_StateMachine_0.HamPack_Anims.AnimStateNode_13.InAirHunker
  nodes [Output Animation Pose(Result=Sequence Player 'HamPack_Idle'() {sequence=&HamPack_Idle})]
 }
 ToSkyDiving {
  path=AnimGraph.AnimGraphNode_StateMachine_0.HamPack_Anims.AnimStateNode_16.ToSkyDiving
  nodes [
   Sequence Player 'HamPack_ToCrouch'() {sequence=/BRCosmetics/Accessories/FORT_Backpacks/Backpack_Pets_Hamster/Animation/HamPack_ToCrouch}
   Output Animation Pose(Result=Sequence Player 'HamPack_Idle'() {sequence=&HamPack_Idle})
  ]
 }
 SkyDiving {
  path=AnimGraph.AnimGraphNode_StateMachine_0.HamPack_Anims.AnimStateNode_18.SkyDiving
  nodes [Output Animation Pose(Result=Sequence Player 'HamPack_Idle'() {sequence=&HamPack_Idle})]
 }
 ToCrouch {
  path=AnimGraph.AnimGraphNode_StateMachine_0.HamPack_Anims.AnimStateNode_2.ToCrouch
  nodes [Output Animation Pose(Result=Sequence Player 'HamPack_ToCrouch'() {sequence=&HamPack_ToCrouch})]
 }
 SprintToIdle {
  path=AnimGraph.AnimGraphNode_StateMachine_0.HamPack_Anims.AnimStateNode_3.SprintToIdle
  nodes [Output Animation Pose(Result=Sequence Player 'HamPack_LocomotionToIdle'() {sequence=/BRCosmetics/Accessories/FORT_Backpacks/Backpack_Pets_Hamster/Animation/HamPack_LocomotionToIdle})]
 }
 CrouchIdle {
  path=AnimGraph.AnimGraphNode_StateMachine_0.HamPack_Anims.AnimStateNode_4.CrouchIdle
  nodes [Output Animation Pose(Result=Sequence Player 'HamPack_Idle'() {sequence=&HamPack_Idle})]
 }
 CrouchToIdle {
  path=AnimGraph.AnimGraphNode_StateMachine_0.HamPack_Anims.AnimStateNode_5.CrouchToIdle
  nodes [Output Animation Pose(Result=Sequence Player 'HamPack_CrouchIdleToIdle'() {sequence=/BRCosmetics/Accessories/FORT_Backpacks/Backpack_Pets_Hamster/Animation/HamPack_CrouchIdleToIdle})]
 }
 ADS {
  path=AnimGraph.AnimGraphNode_StateMachine_0.HamPack_Anims.AnimStateNode_6.ADS
  nodes [Output Animation Pose(Result=Sequence Player 'HamPack_Idle'() {sequence=&HamPack_Idle})]
 }
 CrouchLocomotion {
  path=AnimGraph.AnimGraphNode_StateMachine_0.HamPack_Anims.AnimStateNode_7.CrouchLocomotion
  nodes [Output Animation Pose(Result=Sequence Player 'HamPack_CrouchLocomotion'() {sequence=/BRCosmetics/Accessories/FORT_Backpacks/Backpack_Pets_Hamster/Animation/HamPack_CrouchLocomotion})]
 }
 To CrouchIdle {
  path=AnimGraph.AnimGraphNode_StateMachine_0.HamPack_Anims.AnimStateNode_8.To CrouchIdle
  nodes [Output Animation Pose(Result=Sequence Player 'HamPack_LocoToCrouchIdle'() {sequence=&HamPack_LocoToCrouchIdle})]
 }
 Hunker {
  path=AnimGraph.AnimGraphNode_StateMachine_0.HamPack_Anims.AnimStateNode_9.Hunker
  nodes [Output Animation Pose(Result=Sequence Player 'HamPack_CrouchLocomotion'() {sequence=&HamPack_CrouchLocomotion})]
 }
 Transition {
  path=AnimGraph.AnimGraphNode_StateMachine_0.HamPack_Anims.AnimStateTransitionNode_0.Transition
  nodes [
   n1 = Get bIsDBNO()
   n2 = Get bIsSkyDiving()
   Result(bCanEnterTransition=AND Boolean!(A=NOT Boolean(A=@n2), B=Get IsMovingLocal() -> (IsMovingLocal:bool), C:bool=NOT Boolean(A=@n1)) -> (ReturnValue))
  ]
 }
 Transition {
  path=AnimGraph.AnimGraphNode_StateMachine_0.HamPack_Anims.AnimStateTransitionNode_1.Transition
  nodes [Result(bCanEnterTransition=OR Boolean(A=Get CrouchSprinting() -> (CrouchSprinting:bool), B=Get IsCrouching() -> (IsCrouching:bool)))]
 }
 Transition {
  path=AnimGraph.AnimGraphNode_StateMachine_0.HamPack_Anims.AnimStateTransitionNode_10.Transition
  nodes [Result(bCanEnterTransition=Get bIsInAirNotJumping() -> (bIsInAirNotJumping:bool))]
 }
 Transition {
  path=AnimGraph.AnimGraphNode_StateMachine_0.HamPack_Anims.AnimStateTransitionNode_11.Transition
  nodes [Result(bCanEnterTransition=Get IsCrouching())]
 }
 Transition {
  path=AnimGraph.AnimGraphNode_StateMachine_0.HamPack_Anims.AnimStateTransitionNode_12.Transition
  nodes [
   n1 = Get bIsDBNO()
   Result(bCanEnterTransition=AND Boolean!(A=Get IsMovingLocal(), B=NOT Boolean(A=@n1)) -> (ReturnValue))
  ]
 }
 Transition {
  path=AnimGraph.AnimGraphNode_StateMachine_0.HamPack_Anims.AnimStateTransitionNode_13.Transition
  nodes [Result(bCanEnterTransition=NOT Boolean(A=Get bIsInAirNotJumping()))]
 }
 Transition {
  path=AnimGraph.AnimGraphNode_StateMachine_0.HamPack_Anims.AnimStateTransitionNode_14.Transition
  nodes [Result(bCanEnterTransition=NOT Boolean(A=Get IsMovingLocal()))]
 }
 Transition {
  path=AnimGraph.AnimGraphNode_StateMachine_0.HamPack_Anims.AnimStateTransitionNode_15.Transition
  nodes [Result(bCanEnterTransition=Get bIsDBNO())]
 }
 Transition {
  path=AnimGraph.AnimGraphNode_StateMachine_0.HamPack_Anims.AnimStateTransitionNode_17.Transition
  nodes [Result(bCanEnterTransition=Get CrouchSprinting())]
 }
 Transition {path=AnimGraph.AnimGraphNode_StateMachine_0.HamPack_Anims.AnimStateTransitionNode_18.Transition, nodes [Result()]}
 Transition {
  path=AnimGraph.AnimGraphNode_StateMachine_0.HamPack_Anims.AnimStateTransitionNode_19.Transition
  nodes [Result(bCanEnterTransition=Get CrouchSprinting())]
 }
 Transition {
  path=AnimGraph.AnimGraphNode_StateMachine_0.HamPack_Anims.AnimStateTransitionNode_2.Transition
  nodes [Result(bCanEnterTransition=NOT Boolean(A=Get IsMovingLocal()))]
 }
 Transition {
  path=AnimGraph.AnimGraphNode_StateMachine_0.HamPack_Anims.AnimStateTransitionNode_20.Transition
  nodes [Result(bCanEnterTransition=NOT Boolean(A=Get CrouchSprinting()))]
 }
 Transition {path=AnimGraph.AnimGraphNode_StateMachine_0.HamPack_Anims.AnimStateTransitionNode_21.Transition, nodes [Result()]}
 Transition {
  path=AnimGraph.AnimGraphNode_StateMachine_0.HamPack_Anims.AnimStateTransitionNode_22.Transition
  nodes [Result(bCanEnterTransition=Get CrouchSprinting())]
 }
 Transition {
  path=AnimGraph.AnimGraphNode_StateMachine_0.HamPack_Anims.AnimStateTransitionNode_26.Transition
  nodes [Result(bCanEnterTransition=NOT Boolean(A=Get IsCrouching()))]
 }
 Transition {
  path=AnimGraph.AnimGraphNode_StateMachine_0.HamPack_Anims.AnimStateTransitionNode_27.Transition
  nodes [Result(bCanEnterTransition=Get IsCrouching())]
 }
 Transition {
  path=AnimGraph.AnimGraphNode_StateMachine_0.HamPack_Anims.AnimStateTransitionNode_28.Transition
  nodes [Result(bCanEnterTransition=Get IsHunker() -> (IsHunker:bool))]
 }
 Transition {
  path=AnimGraph.AnimGraphNode_StateMachine_0.HamPack_Anims.AnimStateTransitionNode_29.Transition
  nodes [Result(bCanEnterTransition=Get bIsSkyDiving())]
 }
 Transition {path=AnimGraph.AnimGraphNode_StateMachine_0.HamPack_Anims.AnimStateTransitionNode_3.Transition, nodes [Result()]}
 Transition {
  path=AnimGraph.AnimGraphNode_StateMachine_0.HamPack_Anims.AnimStateTransitionNode_30.Transition
  nodes [Result(bCanEnterTransition=NOT Boolean(A=Get IsHunker()))]
 }
 Transition {path=AnimGraph.AnimGraphNode_StateMachine_0.HamPack_Anims.AnimStateTransitionNode_31.Transition, nodes [Result()]}
 Transition {
  path=AnimGraph.AnimGraphNode_StateMachine_0.HamPack_Anims.AnimStateTransitionNode_33.Transition
  nodes [Result(bCanEnterTransition=NOT Boolean(A=Get IsCrouching()))]
 }
 Transition {
  path=AnimGraph.AnimGraphNode_StateMachine_0.HamPack_Anims.AnimStateTransitionNode_34.Transition
  nodes [Result(bCanEnterTransition=Get IsTargetingLocal() -> (IsTargetingLocal:bool))]
 }
 Transition {
  path=AnimGraph.AnimGraphNode_StateMachine_0.HamPack_Anims.AnimStateTransitionNode_35.Transition
  nodes [Result(bCanEnterTransition=NOT Boolean(A=Get IsTargetingLocal()))]
 }
 Transition {
  path=AnimGraph.AnimGraphNode_StateMachine_0.HamPack_Anims.AnimStateTransitionNode_36.Transition
  nodes [Result(bCanEnterTransition=Get IsTargetingLocal())]
 }
 Transition {
  path=AnimGraph.AnimGraphNode_StateMachine_0.HamPack_Anims.AnimStateTransitionNode_37.Transition
  nodes [Result(bCanEnterTransition=NOT Boolean(A=Get IsCrouching()))]
 }
 Transition {
  path=AnimGraph.AnimGraphNode_StateMachine_0.HamPack_Anims.AnimStateTransitionNode_4.Transition
  nodes [Result(bCanEnterTransition=Get bIsTargeting())]
 }
 Transition {path=AnimGraph.AnimGraphNode_StateMachine_0.HamPack_Anims.AnimStateTransitionNode_41.Transition, nodes [Result()]}
 Transition {
  path=AnimGraph.AnimGraphNode_StateMachine_0.HamPack_Anims.AnimStateTransitionNode_42.Transition
  nodes [Result(bCanEnterTransition=Get IsMovingLocal())]
 }
 Transition {
  path=AnimGraph.AnimGraphNode_StateMachine_0.HamPack_Anims.AnimStateTransitionNode_43.Transition
  nodes [Result(bCanEnterTransition=Get bIsSkyDiving())]
 }
 Transition {
  path=AnimGraph.AnimGraphNode_StateMachine_0.HamPack_Anims.AnimStateTransitionNode_5.Transition
  nodes [Result(bCanEnterTransition=NOT Boolean(A=Get IsCrouching()))]
 }
 Transition {path=AnimGraph.AnimGraphNode_StateMachine_0.HamPack_Anims.AnimStateTransitionNode_6.Transition, nodes [Result()]}
 Transition {
  path=AnimGraph.AnimGraphNode_StateMachine_0.HamPack_Anims.AnimStateTransitionNode_7.Transition
  nodes [Result(bCanEnterTransition=NOT Boolean(A=Get bIsTargeting()))]
 }
 Transition {
  path=AnimGraph.AnimGraphNode_StateMachine_0.HamPack_Anims.AnimStateTransitionNode_8.Transition
  nodes [Result(bCanEnterTransition=Get bIsInAirNotJumping())]
 }
 Transition {
  path=AnimGraph.AnimGraphNode_StateMachine_0.HamPack_Anims.AnimStateTransitionNode_9.Transition
  nodes [Result(bCanEnterTransition=NOT Boolean(A=Get IsCrouching()))]
 }
 EventGraph {nodes [
   Event BlueprintUpdateAnimation()
   > n1 = Cast To FortPlayerPawn(Object=GetOwningActor())
   > n2 = Sequence()
   > n3 = Cast To Fortnite_M_Avg_Player_AnimBlueprint!(execute, Object=GetAnimInstance(self=Get Mesh(self=@n1))) -> (then, CastFailed, AsFortnite M Avg Player Anim Blueprint:Fortnite M Avg Player Anim Blueprint Object Reference) {is=title K2Node_DynamicCast}
   > Set IsMovingLocal(IsMovingLocal:bool=Get bIsMoving2D!(self:Fort Player Anim Instance Object Reference=@n3) -> (bIsMoving2D)) -> (Output_Get:bool)
   > Set IsJumping(IsJumping:bool=Get bIsJumping!(self:Fort Anim Instance Object Reference=@n3) -> (bIsJumping)) -> (Output_Get:bool)
   > Set IsCrouching(IsCrouching:bool=Get bIsCrouching!(self:Fort Player Anim Instance Object Reference=@n3) -> (bIsCrouching)) -> (Output_Get:bool)
   > Set CrouchSprinting(CrouchSprinting:bool=Get bIsCrouchSprinting!(self:Fort Player Anim Instance Object Reference=@n3) -> (bIsCrouchSprinting)) -> (Output_Get:bool)
   > Set IsTargetingLocal(IsTargetingLocal:bool=Get bIsTargeting!(self:Fort Player Anim Instance Object Reference=@n3) -> (bIsTargeting)) -> (Output_Get:bool)
   n4 = Get MeleeTwistCurveValue!(self:Fort Player Anim Instance Object Reference=@n3) -> (MeleeTwistCurveValue:float)
   > Set IsHunker(IsHunker:bool=OR Boolean(A="Not Equal (Float)"!(A:double=@n4, B:double=0) -> (ReturnValue:bool) {is=title K2Node_CallFunction}, B=Get IsDancing() -> (IsDancing:bool))) -> (Output_Get:bool)
   Branch(execute=@n2.then_1, Condition=IsDrivingVehicle!(self:Fort Player Pawn Object Reference=@n1) -> (ReturnValue:bool))
   > Set IsCrouching(IsCrouching=true)
   > Set IsMovingLocal(IsMovingLocal=false)
   n5 = Event OnNotifyPartsStarted!() -> (OutputDelegate, then, Context:Name=None) {is=title K2Node_Event}
   > n6 = Branch(Condition="Equal (Name)"!(A:Name=@n5.Context, B:Name=PlayingBreak1) -> (ReturnValue:bool) {is=title K2Node_CallFunction})
   > Set PlaySprintBreak1(PlaySprintBreak1:bool=true) -> (Output_Get:bool)
   Branch(execute=@n6.else, Condition="Equal (Name)"(A=@n5.Context, B=PlayingBreak2))
   > Set PlaySprintBreak2(PlaySprintBreak2:bool=true) -> (Output_Get:bool)
   n7 = Event OnNotifyPartsStopped!() -> (OutputDelegate, then, Context:Name=None) {is=title K2Node_Event}
   > n8 = Branch(Condition="Equal (Name)"(A=@n7.Context, B=PlayingBreak1))
   > Set PlaySprintBreak1(PlaySprintBreak1=false)
   Branch(execute=@n8.else, Condition="Equal (Name)"(A=@n7.Context, B=PlayingBreak2))
   > Set PlaySprintBreak2(PlaySprintBreak2=false)
  ]}
]
variables [
 CrouchSprinting {type=boolean}
 IsCrouching {type=boolean}
 IsDancing {type=boolean}
 IsHunker {type=boolean}
 IsJumping {type=boolean}
 IsMovingLocal {type=boolean}
 IsTargetingLocal {type=boolean}
 PlaySprintBreak1 {type=boolean}
 PlaySprintBreak2 {type=boolean}
]
